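# Sound cue for the Whisper Remedy Sweet sprite's laser eyes shoot voice line.
# A random node picks one of four equally weighted inventory "Impulsed" VO waves and is excluded from branch culling.
# A modulator at the root adds slight volume variation.
# It plays as a creature vocalization with close sprite VO attenuation and a shared sprite music-vox concurrency set.
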SoundCue /Sprites_WhisperRemedySweetSprite/Assets/Sounds/Cues/VO/Laser_Eyes/SC_VO_S33_WhisperRemedySweet_Shoot
root=Modulator_0
nodes [
 {id=Modulator_0, children [Random_2], properties {volumeMin=0.9}}
 {
  id=Random_2
  children [WavePlayer_13, WavePlayer_14, WavePlayer_15, WavePlayer_17]
  properties {bShouldExcludeFromBranchCulling=true, weights [1,1,1,1]}
 }
 {
  id=WavePlayer_13
  properties {soundWaveAssetPtr=/Sprites_WhisperRemedySweetSprite/Assets/Sounds/Waves/VO/VO_S33_WhisperRemedySweetSprite_Inventory_Impulsed_011}
 }
 {
  id=WavePlayer_14
  properties {soundWaveAssetPtr=/Sprites_WhisperRemedySweetSprite/Assets/Sounds/Waves/VO/VO_S33_WhisperRemedySweetSprite_Inventory_Impulsed_014}
 }
 {
  id=WavePlayer_15
  properties {soundWaveAssetPtr=/Sprites_WhisperRemedySweetSprite/Assets/Sounds/Waves/VO/VO_S33_WhisperRemedySweetSprite_Inventory_Impulsed_015}
 }
 {
  id=WavePlayer_17
  properties {soundWaveAssetPtr=/Sprites_WhisperRemedySweetSprite/Assets/Sounds/Waves/VO/VO_S33_WhisperRemedySweetSprite_Inventory_Impulsed_017}
 }
]
properties {
 attenuationSettings=/Sprites_WhisperRemedySweetSprite/Assets/Sounds/Settings/AS_Sprite_VO_Close
 maxDistance=4000
 priority=3
 soundClassObject=/Game/Sounds/FortSoundClassesAndModes/Fort_Creatures_Vocalization_Normal
 volumeMultiplier=1
 concurrencySet [/Sprites/Assets/Sounds/Music/CS_Sprite_Pri_MusicVox]
}
